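# Looping Niagara system for persistent ambient world effects: butterflies, chromatic bits, fog and screen sprites, airborne leaves, with these emitters all disabled here.
# GPU particles are placed on terrain by a Fortnite spawn module reading the user Landscape and Apollo whole-map color textures.
# Curl noise and looping curves drive drift and gusts.
# Chromatic emitters tint or dim by time of day; bits and fog also fade by camera distance.
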NiagaraSystem /FNGameplayCues/Athena/Effects/Particles/Environments/Generic/NS_Loop_WorldFX_Persistent
dependencies {
 usedDataInterfaces [
  &NiagaraDataInterfaceCamera
  &NiagaraDataInterfaceCurve
  /Script/Niagara.NiagaraDataInterfaceTexture
  &NiagaraDataInterfaceVector2DCurve
  &NiagaraDataInterfaceColorCurve
 ]
 usedDynamicInputs []
 usedModules [
  &SystemState
  &self:SystemSpawnScript.SystemScriptSource.SystemScriptGraph.NiagaraNodeAssignment_0.SetVariables_#1
  &self:SystemSpawnScript.SystemScriptSource.SystemScriptGraph.NiagaraNodeAssignment_1.SetVariables_#2
 ]
}
emitters {
 Butterflies {
  bEnabled=false
  simTarget=GPUComputeSim
  emitterUpdateScript [
   /SetVariables(
    Emitter.CameraPostion:Vector3f=&ReturnCameraVector(Camera Property=::NewEnumerator0, Coordinate Space=::Simulation)
    Emitter.Emitter_HeightOffset:NiagaraFloat=$System.System_HeightOffset
    Emitter.Overall_SpeedMult=0.25
    Emitter.SpawnRate_Curve:NiagaraFloat=&FloatFromCurve(
     FloatCurve=DataInterface() {propertyValues {Curve {keys [
         {arriveTangent=17.5837, interpMode=RCIM_Cubic, leaveTangent=17.5838, tangentMode=RCTM_User, time=0, value=0.1}
         {arriveTangent=-0.0619, interpMode=RCIM_Cubic, leaveTangent=-0.0619, tangentMode=RCTM_User, time=0.0926, value=0.0866}
         <at 0.2896 0.0896>
         <at 0.3175 0.264>
         <at 0.3513 0.0896>
         <at 0.6221 0.1>
         <at 1 0.1>
        ]}}}
     CurveIndex=&Multiply_Float(A=$Emitter.NormalizedLoopAge, B=&RandomRangeFloat(Minimum=0.25, Maximum=1))
     Scale Curve=&Lerp_Float(A=1000, B=1700, Alpha=&UniformRangedFloat(Minimum=0.25, Maximum=1))
    )
    Emitter.SpawnSpeed_Curve:NiagaraFloat=&FloatFromCurve(
     FloatCurve=DataInterface() {propertyValues {Curve {keys [
         {arriveTangent=30.2685, interpMode=RCIM_Cubic, leaveTangent=30.2685, tangentMode=RCTM_User, time=0.0009, value=0.0215}
         {arriveTangent=-0.0619, interpMode=RCIM_Cubic, leaveTangent=-0.0619, tangentMode=RCTM_User, time=0.0989, value=0.3515}
         <at 0.297 0.0032>
         <at 0.3192 0.3332>
         <at 0.3525 0.0398>
         <at 0.6171 0.3148>
         <at 0.9939 0.3698>
        ]}}}
     CurveIndex=&Multiply_Float(A=$Emitter.NormalizedLoopAge, B=&RandomRangeFloat(Minimum=1, Maximum=2))
     Scale Curve=&Multiply_Float(
      A=&Lerp_Float(A=5, B=10, Alpha=&UniformRangedFloat(Minimum=0.25, Maximum=1))
      B=1
     )
    )
    Emitter.SpriteRotRate:NiagaraFloat=&FloatFromCurve(
     FloatCurve=DataInterface() {propertyValues {Curve {keys [
         <at 0.0041 0.1315>
         <at 0.1088 1.9664>
         {arriveTangent=-3.3757, interpMode=RCIM_Cubic, leaveTangent=-3.3757, tangentMode=RCTM_User, time=0.3609, value=0.6391}
         {arriveTangent=-1, interpMode=RCIM_Cubic, leaveTangent=-1, tangentMode=RCTM_User, time=0.5614, value=0.4386}
         <at 0.7086 1.6472>
         <at 1 0>
        ]}}}
     CurveIndex=$Emitter.NormalizedLoopAge
     Scale Curve=500
    )
   ) {isSetParameters=true, module=SetVariables_#3}
   &EmitterState(
    Life Cycle Mode=::NewEnumerator1
    Loop Duration=&RandomRangeFloat(Minimum=5, Maximum=10)
    Recalculate Duration Each Loop=true
    UseLoopDelay=false
   )
   &SpawnRate(SpawnRate=$Emitter.SpawnRate_Curve, Use Spawn Probability=false) {enabled=false}
   &SpawnRate(SpawnRate=100, Use Spawn Probability=false) {module=SpawnRate001}
  ]
  particleSpawnScript [
   /SetVariables(
    Particles.HeightOffset:NiagaraFloat=$System.System_HeightOffset
    Particles.UporDown:NiagaraFloat=&UniformAOrBFloat(
     A=&RandomRangeFloat(Minimum=-10, Maximum=-35)
     B=&RandomRangeFloat(Minimum=-75, Maximum=-150)
     Distribution Weight=0.8
    )
   ) {isSetParameters=true, module=SetVariables_#4}
   &InitializeParticle(
    Color=&ColorFromCurve(
     ColorCurve=DataInterface() {propertyValues {
       AlphaCurve {keys [<at 0 1 RCIM_Cubic>,<at 1 1 RCIM_Cubic>]}
       BlueCurve {keys [<at 0.1102 1 RCIM_Cubic>,<at 0.546 1>,<at 0.8291 0 RCIM_Cubic>]}
       GreenCurve {keys [<at 0.1102 1 RCIM_Cubic>,<at 0.546 0.0641>,<at 0.8291 0.173 RCIM_Cubic>]}
       RedCurve {keys [<at 0.1102 1 RCIM_Cubic>,<at 0.546 0.7944>,<at 0.8291 1 RCIM_Cubic>]}
      }}
     CurveIndex=&RandomRangeFloat(Minimum=0, Maximum=1)
    )
    Lifetime Max=5
    Lifetime Min=4
    Lifetime Mode=::NewEnumerator1
    Mesh Scale Mode=::NewEnumerator3
    Mesh Uniform Scale=&Lerp_Float(
     A=0.02
     B=0.04
     Alpha=&FloatFromCurve(
      FloatCurve=DataInterface() {propertyValues {Curve {keys [
          {interpMode=RCIM_Cubic, tangentMode=RCTM_User, time=0, value=0}
          {arriveTangent=0.8761, interpMode=RCIM_Cubic, leaveTangent=0.8761, tangentMode=RCTM_User, time=0.8, value=0.275}
          {arriveTangent=5.9522, interpMode=RCIM_Cubic, leaveTangent=5.9522, tangentMode=RCTM_User, time=1, value=1}
         ]}}}
      CurveIndex=&RandomRangeFloat(Minimum=0, Maximum=1)
     )
    )
    Sprite Rotation Mode=::NewEnumerator2
    UsePositionOffset=false
    Write Material Random=true
   )
   &InitialMeshOrientation(Mesh Axis To Orient=<0 0 1>, Mesh Orientation Mode=::NewEnumerator2, Use Rotation Vector=false)
   &AddVelocity(Velocity=&RandomRangeVector(Minimum=<10 10 25>, Maximum=<50 50 100>))
   /Niagara/Modules/Spawn/Location/SphereLocation(
    MaskSpawnGroup=false
    Module.Override Randomness=false
    Non Uniform Scale=&RandomRangeVector(Minimum=<-1 -1 0>, Maximum=<1 1 0>)
    Offset=&Multiply_VectorByFloat(
     Vector=<1 0 0>
     Float=&FloatFromCurve(FloatCurve=DataInterface() {propertyValues {Curve {keys [<at 0 0>,<at 2 1>]}}}, CurveIndex=$System.LoopedAge, Scale Curve=5000)
    )
    Sphere Radius=&Multiply_Float(
     A=5000
     B=&FloatFromCurve(
      FloatCurve=DataInterface() {propertyValues {Curve {keys [<at 0 0.1>,<at 0.157 0.9198>]}}}
      CurveIndex=$System.NormalizedLoopAge
      Scale Curve=5
     )
    )
    Surface Only=false
   ) {module=SphereLocation001}
   /NiagaraFN/Modules/Spawn/FN_SpawnParticlesOnTerrainNearPlayer_ParticleType_V2(
    ColorTexture=DataInterface() {propertyValues {
      Texture=/Game/Effects/Fort_Effects/Textures/Environments/Apollo/T_WholeMap_Color_D_03
      TextureUserParameter {parameter {name=None, typeDefHandle {registeredTypeIndex=2}, varData []}}
     }}
    Height Texture -x -y:Vector3f=<-130048 -130048 0>
    HeightOffset=100
    HeightTexture +x +y:Vector3f=<130048 130048 0>
    Landscape:NiagaraDataInterfaceLandscape=$User.Landscape
    Player Position:NiagaraPosition=$Particles.Position
    Player Relative Particle Location:Vector2f=&UniformRangedVector2D(Minimum=<0 0>, Maximum=<0 0>)
    SandColor:LinearColor=<rgba 1 0 0 1>
    SnowColor:LinearColor=<rgba 0 0 1 1>
    Threshold=0.02
   )
   &CurlNoiseForce(
    Noise Frequency=25
    Noise Strength=&Multiply_Float(
     A=50
     B=&Multiply_Float(A=&Multiply_Float(A=$Particles.HeightOffset, B=$Emitter.Overall_SpeedMult), B=1)
    )
    Pan Noise=false
   ) {module=CurlNoiseForce001}
   &ApplyInitialForces()
   /Niagara/Modules/Update/Camera/MaintainInCameraParticleScale()
   &CameraOffset(Camera Offset Amount=-200)
   /SetVariables(Particles.GlowColor=&MultiplyLinearColor(
     Linear Color=$Particles.Color
     Scale Factor=&MakeFloat4FromVectorAndFloat("Vector (RGB)"=<1 1 1>, "Float (Alpha)"=0.02)
    )) {isSetParameters=true}
  ]
  particleUpdateScript [
   &ParticleState()
   &PointAttractionForce(
    Attraction Radius=25000
    Attraction Strength=-50
    Attractor Position:Vector3f=$Emitter.CameraPostion
    Attractor Position Offset:Vector3f=<0 0 250>
    Falloff Exponent=0
    Kill Within Radius=false
    Module.Use Falloff=true
    UseAttractorOffset=true
   ) {module=PointAttractionForce001}
   &Drag(Drag=&RandomRangeFloat(Minimum=0.5, Maximum=1), Use Linear Drag=true, Use Rotational Drag=false)
   &GravityForce(Gravity=&Multiply_VectorByFloat(
     Vector=<0 0 1>
     Float=/Game/Effects/Niagara/DynamicInputs/Math/FN_Fit(
      Clamp:NiagaraBool=false
      Value:NiagaraFloat=$Particles.UporDown
      SourceMin:NiagaraFloat=0
      SourceMax:NiagaraFloat=1
      DestMin:NiagaraFloat=0
      DestMax:NiagaraFloat=-0.1
     )
    ))
   &CurlNoiseForce(
    Curl Noise Cone Mask Angle=120
    Mask Curl Noise=true
    Noise Frequency=75
    Noise Strength=&Multiply_Float(A=$Emitter.SpawnSpeed_Curve, B=300)
    Pan Noise=true
    Pan Noise Field=<0 0 1>
   )
   &FN_Sprite_Attributes(
    Color=$Particles.GlowColor
    Create Color Attribute=true
    Create SpriteSize Attribute=true
    SpriteSize=&Multiply_Vector2D_ByFloat(Float=&MakeFloatFromVector(Vector=$Particles.Scale, Channel=::NewEnumerator0), Vector2D=<850 850>)
    Write to Module Namespace=true
   )
   &SolveForcesAndVelocity(Clamp Velocity=false, Limit Acceleration=false)
   &DynamicMaterialParameters(
    Index 0 Param 1=0.2364
    Index 0 Param 4=&Multiply_Float(A=$System.NormalizedLoopAge, B=3)
    Index 1 Param 1=3000
    Index 1 Param 2=1
    Module.Param0WriteEnabled=true
    Module.Param0WriteEnabled1=true
    Module.Param1WriteEnabled=true
    Module.Param1WriteEnabled1=true
    Module.Param2WriteEnabled=true
    Module.Param2WriteEnabled1=true
    Module.Param3WriteEnabled=true
    Module.Param3WriteEnabled1=true
    Write Parameter Index 1=true
   )
   &ScaleMeshSize(Scale Factor=&VectorFromFloat(Value=&FloatFromCurve(FloatCurve=DataInterface() {propertyValues {Curve {keys [
          <at 0 0 RCIM_Cubic>
          {arriveTangent=1.1765, interpMode=RCIM_Cubic, leaveTangent=1.1765, time=0.15, value=1}
          {arriveTangent=-1.1765, interpMode=RCIM_Cubic, leaveTangent=-1.1765, time=0.85, value=1}
          <at 1 0 RCIM_Cubic>
         ]}}})))
  ]
  rendererClasses [&NiagaraSpriteRendererProperties, &NiagaraMeshRendererProperties]
  renderers [
   {rendererClass=&NiagaraSpriteRendererProperties}
   {rendererClass=&NiagaraMeshRendererProperties, rendererIndex=1}
  ]
 }
 ChromaticBits {
  bEnabled=false
  simTarget=GPUComputeSim
  emitterUpdateScript [
   &EmitterState()
   &SpawnRate(SpawnRate=300, Use Spawn Probability=false)
   /SetVariables(Emitter.CameraPosition:Vector3f=&ReturnCameraVector(Camera Property=::NewEnumerator0, Coordinate Space=::Simulation)) {isSetParameters=true, module=SetVariables_#5}
  ]
  particleSpawnScript [
   &InitializeParticle(
    Color=&MakeLinearColorFromVectorAndFloat("Vector (RGB)"=<1 1 1>, "Float (Alpha)"=&RandomRangeFloat(Minimum=0.08, Maximum=0.15))
    Lifetime Max=13
    Lifetime Min=8
    Lifetime Mode=::NewEnumerator1
    Sprite Rotation Mode=::NewEnumerator2
    Sprite Size Mode=::NewEnumerator3
    Uniform Sprite Size=&Lerp_Float(
     A=5
     B=10
     Alpha=&FloatFromCurve(
      FloatCurve=DataInterface() {propertyValues {Curve {keys [
          {interpMode=RCIM_Cubic, tangentMode=RCTM_User, time=0, value=0}
          {arriveTangent=0.5743, interpMode=RCIM_Cubic, leaveTangent=0.5743, tangentMode=RCTM_User, time=0.7, value=0.2}
          {arriveTangent=1.5556, interpMode=RCIM_Cubic, leaveTangent=1.5556, tangentMode=RCTM_User, time=1, value=1}
         ]}}}
      CurveIndex=&RandomRangeFloat(Minimum=0, Maximum=1)
     )
    )
    UsePositionOffset=false
    Write Material Random=true
   )
   &ShapeLocation(
    Cylinder Radius=5000
    MaskSpawnGroup=false
    Override Seed=false
    Rotation Axis=<0 0 1>
    Rotation Mode=::NewEnumerator1
    Shape Primitive=::NewEnumerator1
    Surface Only=false
    UseOwnerScale=true
   )
   &FN_SpawnParticlesOnTerrainNearPlayer_ParticleType_V2(
    ColorTexture=DataInterface() {propertyValues {
      Texture=&T_WholeMap_Color_D_03
      TextureUserParameter {parameter {name=None, typeDefHandle {registeredTypeIndex=2}, varData []}}
     }}
    Height Texture -x -y:Vector3f=<-130048 -130048 0>
    HeightOffset=100
    HeightTexture +x +y:Vector3f=<130048 130048 0>
    Landscape:NiagaraDataInterfaceLandscape=$User.Landscape
    Player Position:NiagaraPosition=$Particles.Position
    Player Relative Particle Location:Vector2f=&UniformRangedVector2D(Minimum=<0 0>, Maximum=<0 0>)
    SandColor:LinearColor=<rgba 1 0 0 1>
    SnowColor:LinearColor=<rgba 0 0 1 1>
    Threshold=0.02
   )
   /SetVariables(Particles.Which:NiagaraInt32=$Particles.ParticleType) {enabled=false, isSetParameters=true, module=SetVariables_#6}
   &AddVelocity(Velocity=&UniformRangedVector(Minimum=<-10 -10 5>, Maximum=<10 10 25>))
   /Niagara/Modules/Update/Forces/CurlNoiseForce(
    Noise Frequency=5
    Noise Quality / Cost=::NewEnumerator0
    Noise Strength=&RandomRangeFloat(Minimum=5, Maximum=30)
    Pan Noise=true
   )
   &ApplyInitialForces()
  ]
  particleUpdateScript [
   &ParticleState(Lifetime:NiagaraFloat=&Multiply_Float(
     A=$Particles.Lifetime
     B=&FloatFromCurve(
      FloatCurve=DataInterface() {propertyValues {Curve {keys [
          {interpMode=RCIM_Cubic, tangentMode=RCTM_User, time=0, value=1}
          {arriveTangent=0.0074, interpMode=RCIM_Cubic, leaveTangent=0.0074, tangentMode=RCTM_User, time=0.1, value=1}
          {arriveTangent=-0.0952, interpMode=RCIM_Cubic, leaveTangent=-0.0952, tangentMode=RCTM_User, time=0.75, value=1}
          <at 0.95 0>
         ]}}}
      CurveIndex=&NormalizeDistanceRange(
       Start Position:NiagaraPosition=Linked() {linkedVariable {name=Emitter.CameraPosition, type=Vector3f}}
       End Position:NiagaraPosition=$Particles.Position
       Distance:NiagaraFloat=3000
       Invert Normalized Range:NiagaraBool=false
      )
     )
    ))
   /SetVariables(Particles.LerpTODColor:LinearColor=&LerpLinearColors(
     StartColor=&MakeLinearColor(R:NiagaraFloat=0.4168, G:NiagaraFloat=0.6708, B:NiagaraFloat=1.07, A:NiagaraFloat=1)
     EndColor=&MakeLinearColor(R:NiagaraFloat=1.5, G:NiagaraFloat=1.3, B:NiagaraFloat=1, A:NiagaraFloat=1)
     LerpFactor=&FNTimeOfDayScalar(Night Scale=0, Morning Scale=1, Evening Scale=1)
    )) {isSetParameters=true, module=SetVariables_#7}
   &ScaleColor(
    Scale Alpha=&FloatFromCurve(
     FloatCurve=DataInterface() {propertyValues {Curve {keys [
         <at 0 0 RCIM_Cubic>
         {arriveTangent=0.1469, interpMode=RCIM_Cubic, leaveTangent=0.1469, tangentMode=RCTM_User, time=0.5, value=1}
         <at 0.85 0>
         <at 0.9 0 RCIM_Cubic>
        ]}}}
     Scale Curve=&FloatFromCurve(
      FloatCurve=DataInterface() {propertyValues {Curve {keys [
          {interpMode=RCIM_Cubic, tangentMode=RCTM_User, time=0, value=0}
          {arriveTangent=0.0074, interpMode=RCIM_Cubic, leaveTangent=0.0074, tangentMode=RCTM_User, time=0.1, value=1}
          {arriveTangent=-0.0952, interpMode=RCIM_Cubic, leaveTangent=-0.0952, tangentMode=RCTM_User, time=0.5, value=1}
          <at 0.95 0>
         ]}}}
      CurveIndex=&NormalizeDistanceRange(
       Start Position:NiagaraPosition=Linked() {linkedVariable {name=Emitter.CameraPosition, type=Vector3f}}
       End Position:NiagaraPosition=$Particles.Position
       Distance:NiagaraFloat=2500
       Invert Normalized Range:NiagaraBool=false
      )
     )
    )
    Scale RGB=&Multiply_VectorByFloat(
     Vector=&MakeVectorFromLinearColorRGB(Color=$Particles.LerpTODColor)
     Float=&FNTimeOfDayScalar(Night Scale=0.5, Morning Scale=0.25, Evening Scale=0.5)
    )
    ScaleA=true
    ScaleRGB=true
   )
   &ScaleSpriteSize(Uniform Curve Sprite Scale=DataInterface() {propertyValues {Curve {keys [
        <at 0 0 RCIM_Cubic>
        {arriveTangent=1.1765, interpMode=RCIM_Cubic, leaveTangent=1.1765, time=0.15, value=1}
        {arriveTangent=-1.1765, interpMode=RCIM_Cubic, leaveTangent=-1.1765, time=0.85, value=1}
        <at 1 0 RCIM_Cubic>
       ]}}})
   &DynamicMaterialParameters(
    Index 0 Param 1=0.25
    Index 0 Param 2=0.25
    Index 0 Param 3=1
    Index 0 Param 4=1
    Index 1 Param 1=&RandomRangeFloat(Minimum=0.03, Maximum=0.07)
    Index 1 Param 2=&Multiply_Float(
     A=&RandomRangeFloat(Minimum=0.2, Maximum=0.8)
     B=&FloatFromCurve(FloatCurve=DataInterface() {propertyValues {Curve {keys [
          {interpMode=RCIM_Cubic, tangentMode=RCTM_User, time=0, value=0}
          {arriveTangent=2.1115, interpMode=RCIM_Cubic, leaveTangent=2.1115, tangentMode=RCTM_User, time=1, value=1}
         ]}}})
    )
    Index 1 Param 3=&RandomRangeFloat(Minimum=0, Maximum=1)
    Index 1 Param 4=&RandomRangeFloat(Minimum=0, Maximum=1)
    Index 2 Param 1=-0.05
    Index 2 Param 2=0.025
    Index 2 Param 3=&RandomRangeFloat(Minimum=3, Maximum=5)
    Index 2 Param 4=&RandomRangeFloat(Minimum=1.4, Maximum=2)
    Index 3 Param 1=500
    Index 3 Param 2=1
    Index 3 Param 3=&RandomRangeFloat(Minimum=0.1, Maximum=0.85)
    Module.Param0WriteEnabled=true
    Module.Param0WriteEnabled1=true
    Module.Param0WriteEnabled2=true
    Module.Param0WriteEnabled3=true
    Module.Param1WriteEnabled=true
    Module.Param1WriteEnabled1=true
    Module.Param1WriteEnabled2=true
    Module.Param1WriteEnabled3=true
    Module.Param2WriteEnabled=true
    Module.Param2WriteEnabled1=true
    Module.Param2WriteEnabled2=true
    Module.Param2WriteEnabled3=true
    Module.Param3WriteEnabled=true
    Module.Param3WriteEnabled1=true
    Module.Param3WriteEnabled2=true
    Module.Param3WriteEnabled3=true
    Write Parameter Index 1=true
    Write Parameter Index 2=true
    Write Parameter Index 3=true
   )
   &Drag(Drag=0.5, Use Linear Drag=true, Use Rotational Drag=false)
   &SolveForcesAndVelocity(Clamp Velocity=false, Limit Acceleration=false)
   &SpriteRotationRate(Rotation Rate=&FloatFromCurve(
     FloatCurve=DataInterface() {propertyValues {Curve {keys [
         {interpMode=RCIM_Cubic, tangentMode=RCTM_User, time=0, value=1}
         {arriveTangent=-2.7504, interpMode=RCIM_Cubic, leaveTangent=-2.7504, tangentMode=RCTM_User, time=1, value=0}
        ]}}}
     Scale Curve=&RandomRangeFloat(Minimum=-50, Maximum=50)
    ))
  ]
  rendererClasses [&NiagaraSpriteRendererProperties]
  renderers [{rendererClass=&NiagaraSpriteRendererProperties}]
 }
 ChromaticFog {
  bEnabled=false
  simTarget=GPUComputeSim
  emitterUpdateScript [
   &EmitterState()
   &SpawnRate(SpawnRate=9, Use Spawn Probability=false)
   /SetVariables(Emitter.CameraPosition:Vector3f=&ReturnCameraVector(Camera Property=::NewEnumerator0, Coordinate Space=::Simulation)) {isSetParameters=true, module=SetVariables_#5}
  ]
  particleSpawnScript [
   /Niagara/Modules/Spawn/Initialization/InitializeParticle(
    Color=<rgba 1 1 1 0.15>
    Lifetime=&UniformRangedFloat(Minimum=14, Maximum=15)
    Module.Write Color=true
    Module.Write Lifetime=true
    Module.Write Mass=true
    Module.Write Material Random=true
    Module.Write Position=false
    Module.Write Scale=false
    Module.Write SpriteRotation=true
    Module.Write SpriteSize=true
    Sprite Rotation=&UniformRangedFloat(Minimum=0, Maximum=10)
    Sprite Size=&RandomRangeVector2D(Minimum=<3500 1800>, Maximum=<4000 2000>)
   )
   &SphereLocation(
    MaskSpawnGroup=false
    Module.Override Randomness=false
    Non Uniform Scale=<1 1 0>
    Sphere Radius=15000
    Surface Only=false
    Surface Only Band Thickness=1
   )
   &CylinderLocation(
    Cylinder Radius=5000
    Lathe Profile=DataInterface() {propertyValues {Curve {keys []}}}
    MaskSpawnGroup=false
    Module.Enable Lathe Profile=false
    Module.Override Randomness=false
    Module.Surface Only=false
   )
   &FN_SpawnParticlesOnTerrainNearPlayer_ParticleType_V2(
    ColorTexture=DataInterface() {propertyValues {
      Texture=/Game/Effects/Fort_Effects/Textures/Environments/Apollo/T_WholeMap_Color_D_02
      TextureUserParameter {parameter {name=None, typeDefHandle {registeredTypeIndex=2}, varData []}}
     }}
    Height Texture -x -y:Vector3f=<-130048 -130048 0>
    HeightOffset=100
    HeightTexture +x +y:Vector3f=<130048 130048 0>
    Landscape:NiagaraDataInterfaceLandscape=$User.Landscape
    Player Position:NiagaraPosition=$Particles.Position
    Player Relative Particle Location:Vector2f=&UniformRangedVector2D(Minimum=<0 0>, Maximum=<0 0>)
    SandColor:LinearColor=<rgba 1 0 0 1>
    SnowColor:LinearColor=<rgba 0 0 1 1>
   )
   /SetVariables(Particles.Which:NiagaraInt32=$Particles.ParticleType) {enabled=false, isSetParameters=true, module=SetVariables_#6}
   &AddVelocity(Velocity=&UniformRangedVector(Minimum=<15 15 15>, Maximum=<15 15 15>))
   &CurlNoiseForce(Noise Frequency=1, Noise Quality / Cost=::NewEnumerator0, Noise Strength=1, Pan Noise=true)
   &ApplyInitialForces()
  ]
  particleUpdateScript [
   &ParticleState(Lifetime:NiagaraFloat=&Multiply_Float(
     A=$Particles.Lifetime
     B=&FloatFromCurve(
      FloatCurve=DataInterface() {propertyValues {Curve {keys [
          {interpMode=RCIM_Cubic, tangentMode=RCTM_User, time=0.05, value=0}
          {arriveTangent=0.0074, interpMode=RCIM_Cubic, leaveTangent=0.0074, tangentMode=RCTM_User, time=0.25, value=1}
          {arriveTangent=-0.0952, interpMode=RCIM_Cubic, leaveTangent=-0.0952, tangentMode=RCTM_User, time=0.75, value=1}
          <at 0.95 0>
         ]}}}
      CurveIndex=&NormalizeDistanceRange(
       Start Position:NiagaraPosition=Linked() {linkedVariable {name=Emitter.CameraPosition, type=Vector3f}}
       End Position:NiagaraPosition=$Particles.Position
       Distance:NiagaraFloat=10000
       Invert Normalized Range:NiagaraBool=false
      )
     )
    ))
   /SetVariables(Particles.LerpTODColor:LinearColor=&LerpLinearColors(
     StartColor=&MakeLinearColor(R:NiagaraFloat=0.4168, G:NiagaraFloat=0.6708, B:NiagaraFloat=1.07, A:NiagaraFloat=1)
     EndColor=&MakeLinearColor(R:NiagaraFloat=1.5, G:NiagaraFloat=1.3, B:NiagaraFloat=1, A:NiagaraFloat=1)
     LerpFactor=&FNTimeOfDayScalar(Night Scale=0, Morning Scale=1, Evening Scale=1)
    )) {isSetParameters=true, module=SetVariables_#7}
   &ScaleColor(
    Scale Alpha=&FloatFromCurve(
     FloatCurve=DataInterface() {propertyValues {Curve {keys [
         <at 0 0 RCIM_Cubic>
         {arriveTangent=0.1469, interpMode=RCIM_Cubic, leaveTangent=0.1469, tangentMode=RCTM_User, time=0.5, value=1}
         <at 0.85 0>
         <at 0.9 0 RCIM_Cubic>
        ]}}}
     Scale Curve=&FloatFromCurve(
      FloatCurve=DataInterface() {propertyValues {Curve {keys [
          {interpMode=RCIM_Cubic, tangentMode=RCTM_User, time=0.0296, value=0.0032}
          {arriveTangent=0.0074, interpMode=RCIM_Cubic, leaveTangent=0.0074, tangentMode=RCTM_User, time=0.4017, value=0.9932}
          {arriveTangent=-0.0952, interpMode=RCIM_Cubic, leaveTangent=-0.0952, tangentMode=RCTM_User, time=0.5, value=1}
          <at 0.95 0>
         ]}}}
      CurveIndex=&NormalizeDistanceRange(
       Start Position:NiagaraPosition=Linked() {linkedVariable {name=Emitter.CameraPosition, type=Vector3f}}
       End Position:NiagaraPosition=$Particles.Position
       Distance:NiagaraFloat=10000
       Invert Normalized Range:NiagaraBool=false
      )
     )
    )
    Scale RGB=&Multiply_VectorByFloat(
     Vector=&MakeVectorFromLinearColorRGB(Color=$Particles.LerpTODColor)
     Float=&FNTimeOfDayScalar(Night Scale=0.4, Morning Scale=0.25, Evening Scale=0.35)
    )
    ScaleA=true
    ScaleRGB=true
   )
   &ScaleSpriteSize(Scale Factor=&Vector2DFromCurve(
     Vector2Curve=DataInterface() {propertyValues {
       XCurve {keys [
         {arriveTangent=6.5871, interpMode=RCIM_Cubic, leaveTangent=2.865, tangentMode=RCTM_User, time=0, value=0.5}
         {interpMode=RCIM_Cubic, tangentMode=RCTM_User, time=0.3581, value=1}
         {arriveTangent=-1.514, interpMode=RCIM_Cubic, leaveTangent=-1.514, tangentMode=RCTM_User, time=1, value=0.5}
        ]}
       YCurve {keys [
         {arriveTangent=6.7383, interpMode=RCIM_Cubic, leaveTangent=2.865, tangentMode=RCTM_User, time=0, value=0.5}
         {interpMode=RCIM_Cubic, tangentMode=RCTM_User, time=0.3581, value=1}
         {arriveTangent=-1.514, interpMode=RCIM_Cubic, leaveTangent=-1.514, tangentMode=RCTM_User, time=1, value=0.5}
        ]}
      }}
     ScaleCurve=&Multiply_Vector2(
      A=<1 1>
      B=&Vector2DFromFloat(Value=&Multiply_Float(
        A=1
        B=&MakeCustomFloatFromBool(
         Bool=&SetBoolByIntComparison(A=2, B=$Particles.ParticleType, Comparison Type=::NewEnumerator4)
         True Float=1
         False Float=0
        )
       ))
     )
    )) {module=ScaleSpriteSize001}
   &ScaleSpriteSize(Scale Factor=&Multiply_Vector2(
     A=<1 1>
     B=&Vector2DFromFloat(Value=&Multiply_Float(
       A=&FloatFromCurve(FloatCurve=DataInterface() {propertyValues {Curve {keys [
            {interpMode=RCIM_Cubic, tangentMode=RCTM_User, time=0, value=0.7}
            {arriveTangent=0.3677, interpMode=RCIM_Cubic, leaveTangent=0.3677, tangentMode=RCTM_User, time=0.2892, value=0.9107}
            {arriveTangent=-1.2812, interpMode=RCIM_Cubic, leaveTangent=-1.2812, tangentMode=RCTM_User, time=0.9809, value=0.9275}
            <at 1 0>
           ]}}})
       B=&FloatFromCurve(
        FloatCurve=DataInterface() {propertyValues {Curve {keys [
            {interpMode=RCIM_Cubic, tangentMode=RCTM_User, time=0, value=0.0005}
            {arriveTangent=0.0074, interpMode=RCIM_Cubic, leaveTangent=0.0074, tangentMode=RCTM_User, time=0.4, value=0.9}
            {arriveTangent=3.1881, interpMode=RCIM_Cubic, leaveTangent=3.1881, tangentMode=RCTM_User, time=0.85, value=1.5}
            <at 0.9 0>
           ]}}}
        CurveIndex=&NormalizeDistanceRange(
         Start Position:NiagaraPosition=Linked() {linkedVariable {name=Emitter.CameraPosition, type=Vector3f}}
         End Position:NiagaraPosition=$Particles.Position
         Distance:NiagaraFloat=10000
         Invert Normalized Range:NiagaraBool=false
        )
       )
      ))
    ))
   &DynamicMaterialParameters(
    Index 0 Param 1=0.25
    Index 0 Param 2=0.25
    Index 0 Param 3=1
    Index 0 Param 4=1
    Index 1 Param 1=&RandomRangeFloat(Minimum=0.03, Maximum=0.07)
    Index 1 Param 2=&RandomRangeFloat(Minimum=0.1, Maximum=0.2)
    Index 1 Param 3=&RandomRangeFloat(Minimum=0, Maximum=1)
    Index 1 Param 4=&RandomRangeFloat(Minimum=0, Maximum=1)
    Index 2 Param 1=-0.02
    Index 2 Param 2=0.03
    Index 2 Param 3=1.7
    Index 2 Param 4=2
    Index 3 Param 1=15000
    Index 3 Param 2=1
    Index 3 Param 3=&RandomRangeFloat(Minimum=0.05, Maximum=0.2)
    Module.Param0WriteEnabled=true
    Module.Param0WriteEnabled1=true
    Module.Param0WriteEnabled2=true
    Module.Param0WriteEnabled3=true
    Module.Param1WriteEnabled=true
    Module.Param1WriteEnabled1=true
    Module.Param1WriteEnabled2=true
    Module.Param1WriteEnabled3=true
    Module.Param2WriteEnabled=true
    Module.Param2WriteEnabled1=true
    Module.Param2WriteEnabled2=true
    Module.Param2WriteEnabled3=true
    Module.Param3WriteEnabled=true
    Module.Param3WriteEnabled1=true
    Module.Param3WriteEnabled2=true
    Module.Param3WriteEnabled3=true
    Write Parameter Index 1=true
    Write Parameter Index 2=true
    Write Parameter Index 3=true
   )
   &SolveForcesAndVelocity(Clamp Velocity=false, Limit Acceleration=false)
  ]
  rendererClasses [&NiagaraSpriteRendererProperties]
  renderers [{rendererClass=&NiagaraSpriteRendererProperties}]
 }
 ChromaticScreen {
  bEnabled=false
  simTarget=GPUComputeSim
  emitterUpdateScript [
   &EmitterState(Enable Distance Culling=true, MaxDistance=2048, Module.bUseMaxDistance=true, Module.bUseMinDistance=false, Scalability Mode=::NewEnumerator1)
   &SpawnRate(SpawnRate=3, Use Spawn Probability=false)
   /SetVariables(Emitter.UVDistort=0.1, Emitter.VelocityCtrl=20) {isSetParameters=true}
  ]
  particleSpawnScript [
   /Niagara/Modules/Spawn/Initialization/V2/InitializeParticle(
    Lifetime Mode=::NewEnumerator1
    Sprite Rotation Mode=::NewEnumerator2
    Sprite Size Mode=::NewEnumerator1
    Uniform Sprite Size Max=35
    Uniform Sprite Size Min=20
    UsePositionOffset=false
    Write Material Random=true
   )
   &ShapeLocation(
    Cylinder Height=200
    Cylinder Radius=&UniformRangedFloat(Minimum=75, Maximum=175)
    MaskSpawnGroup=false
    Offset=<0 0 150>
    Offset Mode=::NewEnumerator0
    Override Seed=false
    Rotation Axis=<0 1 0>
    Rotation Mode=::NewEnumerator1
    Shape Primitive=::NewEnumerator1
    Surface Only=true
    Use Endcaps In Surface Only Mode=false
    UseOwnerScale=true
   ) {enabled=false}
   &ShapeLocation(
    MaskSpawnGroup=false
    Offset=<0 0 150>
    Offset Mode=::NewEnumerator0
    Override Seed=false
    Rotation Axis=<0 1 0>
    Rotation Mode=::NewEnumerator1
    Sphere Radius=&RandomRangeFloat(Minimum=100, Maximum=200)
    Sphere Surface Distribution=0.75
    UseOwnerScale=true
   ) {module=ShapeLocation001}
   &FN_SpawnParticlesOnTerrainNearPlayer_ParticleType_V2(
    ColorTexture=DataInterface() {propertyValues {
      Texture=&T_WholeMap_Color_D_03
      TextureUserParameter {parameter {name=None, typeDefHandle {registeredTypeIndex=2}, varData []}}
     }}
    Height Texture -x -y:Vector3f=<-130048 -130048 0>
    HeightOffset=120
    HeightTexture +x +y:Vector3f=<130048 130048 0>
    Landscape:NiagaraDataInterfaceLandscape=$User.Landscape
    Player Position:NiagaraPosition=$Particles.Position
    Player Relative Particle Location:Vector2f=&UniformRangedVector2D(Minimum=<0 0>, Maximum=<0 0>)
    SandColor:LinearColor=<rgba 1 0 0 1>
    SnowColor:LinearColor=<rgba 0 0 1 1>
    Threshold=0.02
   )
   &AddVelocity(
    Cone Angle=100
    Cone Axis=<0 1 0>
    Module.UseConeFalloff=false
    Module.UseVelDistribution=false
    Override Seed=false
    Velocity Mode=::NewEnumerator2
   ) {module=AddVelocity001}
   &AddVelocity(Velocity=&Multiply_VectorByFloat(Vector=&RandomRangeVector(Minimum=<0 10 0>, Maximum=<0 100 0>), Float=-1))
  ]
  particleUpdateScript [
   &ParticleState()
   &SolveForcesAndVelocity(Clamp Velocity=false, Limit Acceleration=false)
   &ScaleSpriteSizeBySpeed(Max Scale Factor=<3 3>, Min Scale Factor=<0.5 0.5>, Velocity Threshold=1)
   &DynamicMaterialParameters(
    Index 0 Param 1=0.25
    Index 0 Param 2=0.25
    Index 0 Param 3=1
    Index 0 Param 4=1
    Index 1 Param 1=&RandomRangeFloat(Minimum=0, Maximum=0.01)
    Index 1 Param 2=&Multiply_Float(A=&RandomRangeFloat(Minimum=1, Maximum=1), B=1)
    Index 2 Param 1=1
    Index 2 Param 2=1
    Index 2 Param 3=&RandomRangeFloat(Minimum=3, Maximum=5)
    Index 2 Param 4=&RandomRangeFloat(Minimum=1.4, Maximum=2)
    Index 3 Param 1=500
    Index 3 Param 2=1
    Index 3 Param 3=0.1
    Module.Param0WriteEnabled=true
    Module.Param0WriteEnabled1=true
    Module.Param0WriteEnabled2=true
    Module.Param0WriteEnabled3=true
    Module.Param1WriteEnabled=true
    Module.Param1WriteEnabled1=true
    Module.Param1WriteEnabled2=true
    Module.Param1WriteEnabled3=true
    Module.Param2WriteEnabled=true
    Module.Param2WriteEnabled1=true
    Module.Param2WriteEnabled2=true
    Module.Param2WriteEnabled3=true
    Module.Param3WriteEnabled=true
    Module.Param3WriteEnabled1=true
    Module.Param3WriteEnabled2=true
    Module.Param3WriteEnabled3=true
    Write Parameter Index 1=true
    Write Parameter Index 2=true
    Write Parameter Index 3=true
   )
   &ScaleSpriteSize(Uniform Curve Sprite Scale=DataInterface() {propertyValues {Curve {keys [
        {interpMode=RCIM_Cubic, tangentMode=RCTM_User, time=0, value=0.7}
        {interpMode=RCIM_Cubic, tangentMode=RCTM_User, time=0.5, value=1}
        {interpMode=RCIM_Cubic, tangentMode=RCTM_User, time=1, value=0.7}
       ]}}})
   &ScaleColor(
    Scale Alpha=&FloatFromCurve(
     FloatCurve=DataInterface() {propertyValues {Curve {keys [
         <at 0 0 RCIM_Cubic>
         {interpMode=RCIM_Cubic, tangentMode=RCTM_User, time=0.5, value=1}
         {arriveTangentWeight=0.406, interpMode=RCIM_Cubic, tangentMode=RCTM_User, time=1, value=0}
        ]}}}
     Scale Curve=&RandomRangeFloat(Minimum=0.1, Maximum=0.3)
    )
    Scale RGB=&VectorFromFloat(Value=&FNTimeOfDayScalar(Night Scale=0.1, Morning Scale=0.1, Day Scale=0.25, Evening Scale=0.08))
    ScaleA=true
    ScaleRGB=true
   )
   &ScaleColor(
    Scale Alpha=&FloatFromCurve(
     FloatCurve=DataInterface() {propertyValues {Curve {keys [
         <at 0 0 RCIM_Cubic>
         {interpMode=RCIM_Cubic, tangentMode=RCTM_User, time=0.5, value=1}
         {arriveTangentWeight=0.406, interpMode=RCIM_Cubic, tangentMode=RCTM_User, time=1, value=0}
        ]}}}
     Scale Curve=&RandomRangeFloat(Minimum=1, Maximum=1)
    )
    Scale RGB=&VectorFromFloat(Value=&FNTimeOfDayScalar(Night Scale=1, Morning Scale=1, Evening Scale=1))
    ScaleA=true
    ScaleRGB=true
   ) {enabled=false, module=ScaleColor001}
  ]
  rendererClasses [&NiagaraSpriteRendererProperties]
  renderers [{rendererClass=&NiagaraSpriteRendererProperties}]
 }
 LeavesInAir {
  bEnabled=false
  simTarget=GPUComputeSim
  emitterUpdateScript [
   /SetVariables(
    Emitter.CameraPostion:Vector3f=&ReturnCameraVector(Camera Property=::NewEnumerator0, Coordinate Space=::Simulation)
    Emitter.Emitter_HeightOffset:NiagaraFloat=$System.System_HeightOffset
    Emitter.Overall_SpeedMult=0.25
    Emitter.SpawnRate_Curve:NiagaraFloat=&FloatFromCurve(
     FloatCurve=DataInterface() {propertyValues {Curve {keys [
         {arriveTangent=17.5837, interpMode=RCIM_Cubic, leaveTangent=17.5838, tangentMode=RCTM_User, time=0, value=0.1}
         {arriveTangent=-0.0619, interpMode=RCIM_Cubic, leaveTangent=-0.0619, tangentMode=RCTM_User, time=0.0926, value=0.0866}
         <at 0.2896 0.0896>
         <at 0.3175 0.264>
         <at 0.3513 0.0896>
         <at 0.6221 0.1>
         <at 1 0.1>
        ]}}}
     CurveIndex=&Multiply_Float(A=$Emitter.NormalizedLoopAge, B=&RandomRangeFloat(Minimum=0.25, Maximum=1))
     Scale Curve=&Lerp_Float(A=1000, B=1700, Alpha=&UniformRangedFloat(Minimum=0.25, Maximum=1))
    )
    Emitter.SpawnSpeed_Curve:NiagaraFloat=&FloatFromCurve(
     FloatCurve=DataInterface() {propertyValues {Curve {keys [
         {arriveTangent=30.2685, interpMode=RCIM_Cubic, leaveTangent=30.2685, tangentMode=RCTM_User, time=0.0009, value=0.0215}
         {arriveTangent=-0.0619, interpMode=RCIM_Cubic, leaveTangent=-0.0619, tangentMode=RCTM_User, time=0.0989, value=0.3515}
         <at 0.297 0.0032>
         <at 0.3192 0.3332>
         <at 0.3525 0.0398>
         <at 0.6171 0.3148>
         <at 0.9939 0.3698>
        ]}}}
     CurveIndex=&Multiply_Float(A=$Emitter.NormalizedLoopAge, B=&RandomRangeFloat(Minimum=1, Maximum=2))
     Scale Curve=&Multiply_Float(
      A=&Lerp_Float(A=5, B=10, Alpha=&UniformRangedFloat(Minimum=0.25, Maximum=1))
      B=1
     )
    )
    Emitter.SpriteRotRate:NiagaraFloat=&FloatFromCurve(
     FloatCurve=DataInterface() {propertyValues {Curve {keys [
         <at 0.0041 0.1315>
         <at 0.1088 1.9664>
         {arriveTangent=-3.3757, interpMode=RCIM_Cubic, leaveTangent=-3.3757, tangentMode=RCTM_User, time=0.3609, value=0.6391}
         {arriveTangent=-1, interpMode=RCIM_Cubic, leaveTangent=-1, tangentMode=RCTM_User, time=0.5614, value=0.4386}
         <at 0.7086 1.6472>
         <at 1 0>
        ]}}}
     CurveIndex=$Emitter.NormalizedLoopAge
     Scale Curve=500
    )
   ) {isSetParameters=true, module=SetVariables_#3}
   &EmitterState(
    Life Cycle Mode=::NewEnumerator1
    Loop Duration=&RandomRangeFloat(Minimum=5, Maximum=10)
    Recalculate Duration Each Loop=true
    UseLoopDelay=false
   )
   &SpawnRate(SpawnRate=$Emitter.SpawnRate_Curve, Use Spawn Probability=false)
  ]
  particleSpawnScript [
   /SetVariables(
    Particles.HeightOffset:NiagaraFloat=$System.System_HeightOffset
    Particles.UporDown:NiagaraFloat=&UniformAOrBFloat(
     A=&RandomRangeFloat(Minimum=100, Maximum=200)
     B=&RandomRangeFloat(Minimum=1250, Maximum=1500)
     Distribution Weight=0.8
    )
   ) {isSetParameters=true, module=SetVariables_#4}
   /Niagara/Modules/Spawn/Initialization/InitializeParticle(
    Color=&RandomRangeLinearColor(Minimum=<rgba 0.4948 0.0445 0.0199 1>, Maximum=<rgba 0.7604 0.2336 0 1>)
    Lifetime=&UniformRangedFloat(Minimum=4, Maximum=5)
    Module.Write Color=true
    Module.Write Lifetime=true
    Module.Write Mass=true
    Module.Write Material Random=true
    Module.Write Position=false
    Module.Write Scale=false
    Module.Write SpriteRotation=true
    Module.Write SpriteSize=true
    Position=&SimulationPosition()
    Sprite Rotation=&UniformRangedFloat(Minimum=-180, Maximum=180)
    Sprite Size=&Vector2DFromFloat(Value=&UniformRangedFloat(Minimum=25, Maximum=25))
   )
   &AddVelocity(Velocity=&Multiply_VectorByFloat(
     Vector=&Multiply_VectorByFloat(
      Vector=&Multiply_VectorByFloat(Vector=&UniformRangedVector(Minimum=<60 128 100>, Maximum=<120 200 150>), Float=$Emitter.SpawnSpeed_Curve)
      Float=$Emitter.Overall_SpeedMult
     )
     Float=$Particles.HeightOffset
    ))
   &SphereLocation(
    MaskSpawnGroup=false
    Module.Override Randomness=false
    Non Uniform Scale=&RandomRangeVector(Minimum=<-1 -1 0>, Maximum=<1 1 0>)
    Offset=&Multiply_VectorByFloat(
     Vector=<1 0 0>
     Float=&FloatFromCurve(FloatCurve=DataInterface() {propertyValues {Curve {keys [<at 0 0>,<at 2 1>]}}}, CurveIndex=$System.LoopedAge, Scale Curve=5000)
    )
    Sphere Radius=&Multiply_Float(
     A=5000
     B=&FloatFromCurve(
      FloatCurve=DataInterface() {propertyValues {Curve {keys [<at 0 0.1>,<at 0.157 0.9198>]}}}
      CurveIndex=$System.NormalizedLoopAge
      Scale Curve=5
     )
    )
    Surface Only=false
   ) {module=SphereLocation001}
   &FN_SpawnParticlesOnTerrainNearPlayer_ParticleType_V2(
    ColorTexture=DataInterface() {propertyValues {
      Texture=&T_WholeMap_Color_D_02
      TextureUserParameter {parameter {name=None, typeDefHandle {registeredTypeIndex=2}, varData []}}
     }}
    Height Texture -x -y:Vector3f=<-130048 -130048 0>
    HeightOffset=100
    HeightTexture +x +y:Vector3f=<130048 130048 0>
    Landscape:NiagaraDataInterfaceLandscape=$User.Landscape
    Player Position:NiagaraPosition=$Particles.Position
    Player Relative Particle Location:Vector2f=&UniformRangedVector2D(Minimum=<0 0>, Maximum=<0 0>)
    SandColor:LinearColor=<rgba 1 0 0 1>
    SnowColor:LinearColor=<rgba 0 0 1 1>
   )
   /Niagara/Modules/Update/Forces/V2/CurlNoiseForce(
    Noise Strength=&Multiply_Float(
     A=150
     B=&Multiply_Float(A=&Multiply_Float(A=$Particles.HeightOffset, B=$Emitter.Overall_SpeedMult), B=1)
    )
    Pan Noise=false
   ) {module=CurlNoiseForce001}
   &ApplyInitialForces()
   &MaintainInCameraParticleScale()
   &CameraOffset(Camera Offset Amount=-200)
  ]
  particleUpdateScript [
   /SetVariables(Particles.ParticleUpdate_RotRate:NiagaraFloat=&FloatFromCurve(
     FloatCurve=DataInterface() {propertyValues {Curve {keys [
         <at 0.0041 0.1315>
         <at 0.1088 1.9664>
         {arriveTangent=-3.3757, interpMode=RCIM_Cubic, leaveTangent=-3.3757, tangentMode=RCTM_User, time=0.3609, value=0.6391}
         {arriveTangent=-1, interpMode=RCIM_Cubic, leaveTangent=-1, tangentMode=RCTM_User, time=0.5614, value=0.4386}
         <at 0.7086 1.6472>
         {arriveTangent=-1.004, interpMode=RCIM_Cubic, leaveTangent=-1.004, tangentMode=RCTM_User, time=0.8311, value=0.9545}
         {arriveTangent=-5.6531, interpMode=RCIM_Cubic, leaveTangent=-5.6531, tangentMode=RCTM_User, time=0.8987, value=1.4563}
         <at 0.9384 0.3483>
         <at 1 0>
        ]}}}
     Scale Curve=100
    )) {isSetParameters=true}
   &ParticleState()
   /Niagara/Modules/Update/SubUV/SubUVAnimation(Number Of Frames=64)
   &ScaleSpriteSize(Scale Factor=&Vector2DFromCurve(
     Vector2Curve=DataInterface() {propertyValues {
       XCurve {keys [
         {arriveTangent=5.9728, interpMode=RCIM_Cubic, leaveTangent=5.9728, tangentMode=RCTM_User, time=-0.0064, value=0.0304}
         {arriveTangent=-0.1, interpMode=RCIM_Cubic, leaveTangent=-0.1, tangentMode=RCTM_User, time=0.3581, value=0.9069}
         {interpMode=RCIM_Cubic, tangentMode=RCTM_User, time=1, value=0}
        ]}
       YCurve {keys [
         {arriveTangent=6.2715, interpMode=RCIM_Cubic, leaveTangent=6.2715, tangentMode=RCTM_User, time=-0.0064, value=0.0122}
         {arriveTangent=-0.1001, interpMode=RCIM_Cubic, leaveTangent=-0.1001, tangentMode=RCTM_User, time=0.3581, value=0.9359}
         {interpMode=RCIM_Cubic, tangentMode=RCTM_User, time=0.9988, value=0}
        ]}
      }}
     ScaleCurve=&Multiply_Vector2(
      A=<1 1>
      B=&Vector2DFromFloat(Value=&Multiply_Float(
        A=1
        B=&MakeCustomFloatFromBool(
         Bool=&SetBoolByIntComparison(A=2, B=$Particles.ParticleType, Comparison Type=::NewEnumerator4)
         True Float=1
         False Float=0
        )
       ))
     )
    ))
   &ScaleSpriteSize(Scale Factor=&Multiply_Vector2(
     A=<1 1>
     B=&Vector2DFromFloat(Value=&Multiply_Float(
       A=&FloatFromCurve(FloatCurve=DataInterface() {propertyValues {Curve {keys [
            {interpMode=RCIM_Cubic, tangentMode=RCTM_User, time=-0.0061, value=0.489}
            {arriveTangent=0.3677, interpMode=RCIM_Cubic, leaveTangent=0.3677, tangentMode=RCTM_User, time=0.2892, value=0.9107}
            {arriveTangent=-1.2812, interpMode=RCIM_Cubic, leaveTangent=-1.2812, tangentMode=RCTM_User, time=0.9809, value=0.9275}
            <at 1 0>
           ]}}})
       B=&FloatFromCurve(
        FloatCurve=DataInterface() {propertyValues {Curve {keys [
            {interpMode=RCIM_Cubic, tangentMode=RCTM_User, time=0, value=0.0005}
            {arriveTangent=0.0074, interpMode=RCIM_Cubic, leaveTangent=0.0074, tangentMode=RCTM_User, time=0.15, value=1.0236}
            {arriveTangent=-0.0952, interpMode=RCIM_Cubic, leaveTangent=-0.0952, tangentMode=RCTM_User, time=0.85, value=1}
            <at 0.9 0>
           ]}}}
        CurveIndex=&NormalizeDistanceRange(
         Start Position:NiagaraPosition=Linked() {linkedVariable {name=Emitter.CameraPostion, type=Vector3f}}
         End Position:NiagaraPosition=$Particles.Position
         Distance:NiagaraFloat=10000
         Invert Normalized Range:NiagaraBool=false
        )
        Scale Curve=0.85
       )
      ))
    )) {module=ScaleSpriteSize001}
   &PointAttractionForce(
    Attraction Radius=25000
    Attraction Strength=-150
    Attractor Position:Vector3f=$Emitter.CameraPostion
    Attractor Position Offset:Vector3f=<0 0 250>
    Falloff Exponent=0
    Kill Within Radius=false
    Module.Use Falloff=true
    UseAttractorOffset=true
   ) {module=PointAttractionForce001}
   &Drag(
    Drag=&FloatFromCurve(FloatCurve=DataInterface() {propertyValues {Curve {keys [<at -0.0085 0.0857>,<at 0.0429 0.6046>,<at 0.1003 0.0765>,<at 1 0>]}}})
    Use Linear Drag=true
    Use Rotational Drag=false
   )
   &GravityForce(Gravity=&Multiply_VectorByFloat(
     Vector=<0 0 1>
     Float=&FN_Fit(
      Clamp:NiagaraBool=false
      Value:NiagaraFloat=$Particles.UporDown
      SourceMin:NiagaraFloat=0
      SourceMax:NiagaraFloat=1
      DestMin:NiagaraFloat=0
      DestMax:NiagaraFloat=-0.1
     )
    ))
   &CurlNoiseForce(
    Curl Noise Cone Mask Angle=120
    Mask Curl Noise=true
    Noise Frequency=250
    Noise Strength=&Multiply_Float(A=$Emitter.SpawnSpeed_Curve, B=500)
    Pan Noise=true
    Pan Noise Field=<0 0 1>
   )
   &SpriteRotationRate(Rotation Rate=$Emitter.SpriteRotRate)
   &ScaleColor(
    Scale Alpha=&FloatFromCurve(FloatCurve=DataInterface() {propertyValues {Curve {keys [
         {interpMode=RCIM_Cubic, tangentMode=RCTM_User, time=0, value=0}
         {arriveTangent=-0.0948, interpMode=RCIM_Cubic, leaveTangent=-0.0948, tangentMode=RCTM_User, time=0.1, value=1}
         <at 0.942 0.9825>
         <at 1 0>
        ]}}})
    Scale RGB=&Multiply_VectorByFloat(
     Vector=<1 1 1>
     Float=/Niagara/DynamicInputs/Math/LerpMultipleFloats(
      Float Count:ENiagara_MultipleLerpCount=::NewEnumerator2
      1:NiagaraFloat=0.5
      2:NiagaraFloat=0.85
      3:NiagaraFloat=1
      4:NiagaraFloat=0.85
      5:NiagaraFloat=0.5
      Alpha:NiagaraFloat=Linked() {linkedVariable {name=NPC.FortniteNPC.TimeOfDay_0-1, type=NiagaraFloat}}
     )
    )
    ScaleA=true
    ScaleRGB=true
   )
   /Niagara/Modules/Update/SubUV/V2/SubUVAnimation(SubUV Animation Mode:ENiagara_SubUVLookupMode=::NewEnumerator1, SubUV Loop Count=2) {module=SubUVAnimation001}
   &DynamicMaterialParameters(
    Index 0 Param 1=&Multiply_Float(A=$Particles.ParticleUpdate_RotRate, B=0.0005)
    Index 0 Param 2=&FloatFromCurve(
     FloatCurve=DataInterface() {propertyValues {Curve {keys [
         {interpMode=RCIM_Cubic, tangentMode=RCTM_User, time=0, value=0.1}
         {arriveTangent=1.3207, interpMode=RCIM_Cubic, leaveTangent=1.3207, tangentMode=RCTM_User, time=0.6999, value=0.2782}
         {arriveTangent=3.3434, interpMode=RCIM_Cubic, leaveTangent=3.3434, tangentMode=RCTM_User, time=1, value=1}
        ]}}}
     Scale Curve=0.1
    )
    Module.Param0WriteEnabled=true
    Module.Param1WriteEnabled=true
    Module.Param2WriteEnabled=true
    Module.Param3WriteEnabled=true
   )
   &SolveForcesAndVelocity(Clamp Velocity=false, Limit Acceleration=false)
   /NiagaraFN/Modules/Spawn/FN_BlockParticlesOnLandScape(LandscapeHeight:NiagaraFloat=&MakeFloatFromVector(Vector=$Output.FN_SpawnParticlesOnTerrainNearPlayer.LandscapePostion, Channel=::NewEnumerator2)) {enabled=false}
  ]
  rendererClasses [&NiagaraSpriteRendererProperties]
  renderers [{rendererClass=&NiagaraSpriteRendererProperties}]
 }
}
summary {
 systemName=NS_Loop_WorldFX_Persistent
 emitters [
  {bEnabled=false, emitterName=ChromaticFog, simTarget=GPUComputeSim, rendererClasses [&NiagaraSpriteRendererProperties]}
  {bEnabled=false, emitterName=LeavesInAir, simTarget=GPUComputeSim, rendererClasses [&NiagaraSpriteRendererProperties]}
  {
   bEnabled=false
   emitterName=Butterflies
   simTarget=GPUComputeSim
   rendererClasses [&NiagaraSpriteRendererProperties, &NiagaraMeshRendererProperties]
  }
  {bEnabled=false, emitterName=ChromaticBits, simTarget=GPUComputeSim, rendererClasses [&NiagaraSpriteRendererProperties]}
  {bEnabled=false, emitterName=ChromaticScreen, simTarget=GPUComputeSim, rendererClasses [&NiagaraSpriteRendererProperties]}
 ]
 userVariables [User.Landscape {
   type=NiagaraDataInterfaceLandscape
   description=""
   defaultValue {
    struct=NiagaraExt_VariableValue_DataInterface
    value {
     dataInterface=&self:NiagaraDataInterfaceLandscape_0
     dataInterfaceClass=&NiagaraDataInterfaceLandscape
    }
   }
  }]
}
